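# Cape accessory AnimBlueprint with an IcicleLarva variant event that pushes flight-state flags to the player's FortAnimInstance.
# The pose copies from the parent mesh through a deform Control Rig, then feeds cached paths for rigid body, constraint-based AnimDynamics, movement-weighted trail controllers and DBNO trails, selected by bools.
# EventGraph drives Chaos cloth or gravity override from a parameter asset and adjusts trail rotation for skydiving, with state-specific cloth settings such as Ascend.
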AnimBlueprint /BRCosmetics/Accessories/FORT_Capes/Cape_M_MED_WealthLamb_Fate/Meshes/M_MED_WealthLamb_Fate_Cape_AnimBP_IcicleLarva
target_skeleton=&/M_MED_WealthLamb_Fate_Cape_Skeleton
graphs [
 AnimGraph {nodes [
   n1 = "Transform (Modify) Bone - Bone: cape_02_l"(Alpha=1, Translation="0.076,-0.989,2.016", Rotation="0,0,0", Scale="1,1,1") {
    rotationMode=BMM_Additive
    rotationSpace=BCS_BoneSpace
    translationMode=BMM_Additive
    translationSpace=BCS_BoneSpace
    alphaBoolBlend {blendOption=Linear}
    boneToModify {boneName=cape_02_l}
    translation=<0.076 -0.989 2.016>
   }
   "Transform (Modify) Bone - Bone: cape_02_r"(ComponentPose=@n1, Alpha=1, Translation="-0.874,-0.834,3.604", Rotation="0,0,0", Scale="1,1,1") {
    rotationMode=BMM_Additive
    rotationSpace=BCS_BoneSpace
    translationMode=BMM_Additive
    translationSpace=BCS_BoneSpace
    alphaBoolBlend {blendOption=Linear}
    boneToModify {boneName=cape_02_r}
    translation=<-0.874 -0.834 3.604>
   }
   n2 = Blend Poses by bool(BlendPose_0=Use cached pose 'dbnoCache'(), BlendPose_1=Use cached pose 'trailCache'(), bActiveValue=Get bIsDBNO()) {blendType=Linear}
   n3 = Use cached pose 'animDyn Cache'()
   n4 = Get bAnimDynamicsEnabled() -> (bAnimDynamicsEnabled:bool)
   Output Pose(Result=Blend Poses by bool(
     BlendPose_0=Use cached pose 'RbanCache'()
     BlendPose_1=Blend Poses by bool(BlendPose_0=@n3, BlendPose_1=@n2, bActiveValue=@n4) {blendType=Linear}
     bActiveValue=Get bRigidBodyEnabled()
    ) {blendType=Linear}) {layerGroup=None}
   n5 = Control Rig!(Source=Copy Pose From Mesh() {bUseAttachedParent=true}, Alpha, LodSettings:bool=Get Is LOD2-3() -> (Is LOD2-3:bool)) -> (Pose) {
    alphaBoolBlend {blendOption=Linear}
    controlRigAssetReference {blueprintRigClass=&/M_MED_WealthLamb_Fate_Cape_DeformCtrlRig.M_MED_WealthLamb_Fate_Cape_DeformCtrlRig_C}
   }
   Save cached pose 'Start_Cache'(Pose=@n5) {cacheName=Start_Cache}
   n6 = Get TrailControllerRot() -> (TrailControllerRot:Rotator)
   n7 = "Transform (Modify) Bone - Bone: cape_mid_01_l_MSTR"(ComponentPose=Local To Component(LocalPose=Use cached pose 'constraintCache'()), Alpha=1, Translation="0,0,0", Rotation=@n6, Scale="1,1,1") {rotationMode=BMM_Additive, rotationSpace=BCS_BoneSpace, alphaBoolBlend {blendOption=Linear}, boneToModify {boneName=cape_mid_01_l_MSTR}}
   n8 = "Transform (Modify) Bone - Bone: cape_mid_01_r_MSTR"(ComponentPose=@n7, Alpha=1, Translation="0,0,0", Rotation=@n6, Scale="1,1,1") {rotationMode=BMM_Additive, rotationSpace=BCS_BoneSpace, alphaBoolBlend {blendOption=Linear}, boneToModify {boneName=cape_mid_01_r_MSTR}}
   n9 = "Transform (Modify) Bone - Bone: cape_01_l"(ComponentPose=@n8, Alpha=1, Translation="0,0,0", Rotation=@n6, Scale="1,1,1") {rotationMode=BMM_Additive, rotationSpace=BCS_BoneSpace, alphaBoolBlend {blendOption=Linear}, boneToModify {boneName=cape_01_l}}
   n10 = "Transform (Modify) Bone - Bone: cape_01_r"(ComponentPose=@n9, Alpha=1, Translation="0,0,0", Rotation=@n6, Scale="1,1,1") {rotationMode=BMM_Additive, rotationSpace=BCS_BoneSpace, alphaBoolBlend {blendOption=Linear}, boneToModify {boneName=cape_01_r}}
   n11 = Get BwdMovement() -> (BwdMovement:double)
   n12 = Trail controller - Bone: cape_mid_06_l(ComponentPose=@n10, Alpha=@n11) {
    bUsePlanarLimit=true
    chainLength=5
    debugLifeTime=0.1
    maxDeltaTime=0.333
    alphaScaleBiasClamp {bInterpResult=true}
    rotationLimits [{limitMax=<180 180 180>}, {limitMax=<180 180 180>}, {limitMax=<180 180 180>}, {limitMax=<180 180 180>}, {limitMax=<180 180 180>}]
    rotationOffsets [<0 0 0>, <0 0 0>, <0 0 0>, <0 0 0>, <0 0 0>]
    trailBone {boneName=cape_mid_06_l}
    trailRelaxationSpeed {editorCurveData {keys [<at -0.0134 19.3841 RCIM_Linear>,<at 0.5079 12.8986 RCIM_Linear>,<at 0.9955 9.6044 RCIM_Linear>]}}
   }
   n13 = Trail controller - Bone: cape_mid_06_r(ComponentPose=@n12, Alpha=@n11) {
    bUsePlanarLimit=true
    chainLength=5
    debugLifeTime=0.1
    maxDeltaTime=0.0333
    alphaScaleBiasClamp {bInterpResult=true}
    rotationLimits [
     {limitMax=<180 180 180>}
     {limitMax=<180 180 180>}
     {limitMax=<180 90 180>, limitMin=<-180 -90 -180>}
     {limitMax=<180 180 180>}
     {limitMax=<180 180 180>}
    ]
    rotationOffsets [<0 0 0>, <0 0 0>, <0 0 0>, <0 0 0>, <0 0 0>]
    trailBone {boneName=cape_mid_06_r}
    trailRelaxationSpeed {editorCurveData {keys [<at 0 15 RCIM_Linear>,<at 0.5289 13.1838 RCIM_Linear>,<at 1 6 RCIM_Linear>]}}
   }
   n14 = Trail controller - Bone: cape_07_l(ComponentPose=@n13, Alpha=@n11) {
    bUsePlanarLimit=true
    chainLength=5
    debugLifeTime=0.1
    maxDeltaTime=0.0333
    alphaScaleBiasClamp {bInterpResult=true}
    rotationLimits [
     {limitMax=<180 180 180>}
     {limitMax=<180 180 180>}
     {limitMax=<180 90 180>, limitMin=<-180 -90 -180>}
     {limitMax=<180 180 180>}
     {limitMax=<180 180 180>}
    ]
    rotationOffsets [<0 0 0>, <0 0 0>, <0 0 0>, <0 0 0>, <0 0 0>]
    trailBone {boneName=cape_07_l}
    trailRelaxationSpeed {editorCurveData {keys [<at 0 15 RCIM_Linear>,<at 0.5289 13.1838 RCIM_Linear>,<at 1 6 RCIM_Linear>]}}
   }
   n15 = Trail controller - Bone: cape_07_r(ComponentPose=@n14, Alpha=@n11) {
    bUsePlanarLimit=true
    chainLength=5
    debugLifeTime=0.1
    maxDeltaTime=0.0333
    alphaScaleBiasClamp {bInterpResult=true}
    rotationLimits [
     {limitMax=<180 180 180>}
     {limitMax=<180 180 180>}
     {limitMax=<180 90 180>, limitMin=<-180 -90 -180>}
     {limitMax=<180 180 180>}
     {limitMax=<180 180 180>}
    ]
    rotationOffsets [<0 0 0>, <0 0 0>, <0 0 0>, <0 0 0>, <0 0 0>]
    trailBone {boneName=cape_07_r}
    trailRelaxationSpeed {editorCurveData {keys [<at 0 15 RCIM_Linear>,<at 0.5289 13.1838 RCIM_Linear>,<at 1 6 RCIM_Linear>]}}
   }
   Save cached pose 'trailCache'(Pose=Component To Local(ComponentPose=@n15)) {cacheName=trailCache}
   n16 = "Transform (Modify) Bone - Bone: cape_01_l"(ComponentPose=Local To Component(LocalPose=Use cached pose 'Start_Cache'()), Alpha=1, Translation="0,0,0", Rotation="5,0,0", Scale="1,1,1") {
    rotationMode=BMM_Additive
    rotationSpace=BCS_BoneSpace
    alphaBoolBlend {blendOption=Linear}
    boneToModify {boneName=cape_01_l}
    rotation {pitch=5}
   }
   n17 = "Transform (Modify) Bone - Bone: cape_03_l"(ComponentPose=@n16, Alpha=1, Translation="0,0,0", Rotation="-10,0,0", Scale="1,1,1") {
    rotationMode=BMM_Additive
    rotationSpace=BCS_BoneSpace
    alphaBoolBlend {blendOption=Linear}
    boneToModify {boneName=cape_03_l}
    rotation {pitch=-10}
   }
   n18 = "Transform (Modify) Bone - Bone: cape_06_l"(ComponentPose=@n17, Alpha=1, Translation="0,0,0", Rotation="-25,0,0", Scale="1,1,1") {
    rotationMode=BMM_Additive
    rotationSpace=BCS_BoneSpace
    alphaBoolBlend {blendOption=Linear}
    boneToModify {boneName=cape_06_l}
    rotation {pitch=-25}
   }
   n19 = Trail controller - Bone: cape_mid_06_l!(ComponentPose=@n18, Alpha=1, RelaxationSpeedScale:float=2, FakeVelocity:Vector="0,0,0") -> (Pose) {
    bActorSpaceFakeVel=true
    bShowTrailLocation=true
    bUsePlanarLimit=true
    chainLength=5
    debugLifeTime=0.1
    maxDeltaTime=0.0333
    relaxationSpeedScale=1.2
    alphaScaleBiasClamp {bInterpResult=true, interpSpeedDecreasing=25, interpSpeedIncreasing=25}
    baseJoint {boneName=backpack_root}
    rotationLimits [{limitMax=<180 180 180>}, {limitMax=<180 180 180>}, {limitMax=<180 180 180>}, {limitMax=<180 180 180>}, {limitMax=<180 180 180>}]
    rotationOffsets [<0 0 0>, <0 0 0>, <0 0 0>, <0 0 0>, <0 0 0>]
    trailBone {boneName=cape_mid_06_l}
    trailRelaxationSpeed {editorCurveData {keys [<at 0 15 RCIM_Linear>,<at 1 6 RCIM_Linear>]}}
   }
   n20 = Trail controller - Bone: cape_mid_06_r!(ComponentPose=@n19, Alpha=1, RelaxationSpeedScale:float=2, FakeVelocity:Vector="0,0,0") -> (Pose) {
    bActorSpaceFakeVel=true
    bShowTrailLocation=true
    bUsePlanarLimit=true
    chainLength=5
    debugLifeTime=0.1
    maxDeltaTime=0.0333
    relaxationSpeedScale=1.2
    alphaScaleBiasClamp {bInterpResult=true, interpSpeedDecreasing=25, interpSpeedIncreasing=25}
    baseJoint {boneName=backpack_root}
    rotationLimits [{limitMax=<180 180 180>}, {limitMax=<180 180 180>}, {limitMax=<180 180 180>}, {limitMax=<180 180 180>}, {limitMax=<180 180 180>}]
    rotationOffsets [<0 0 0>, <0 0 0>, <0 0 0>, <0 0 0>, <0 0 0>]
    trailBone {boneName=cape_mid_06_r}
    trailRelaxationSpeed {editorCurveData {keys [<at 0 15 RCIM_Linear>,<at 1 6 RCIM_Linear>]}}
   }
   n21 = Trail controller - Bone: cape_07_l!(ComponentPose=@n20, Alpha=1, RelaxationSpeedScale:float=2, FakeVelocity:Vector="0,0,0") -> (Pose) {
    bActorSpaceFakeVel=true
    bShowTrailLocation=true
    bUsePlanarLimit=true
    chainLength=6
    debugLifeTime=0.1
    maxDeltaTime=0.0333
    relaxationSpeedScale=1.2
    alphaScaleBiasClamp {bInterpResult=true, interpSpeedDecreasing=25, interpSpeedIncreasing=25}
    baseJoint {boneName=backpack_root}
    rotationLimits [
     {limitMax=<180 180 180>}
     {limitMax=<180 180 180>}
     {limitMax=<180 180 180>}
     {limitMax=<180 180 180>}
     {limitMax=<180 180 180>}
     {limitMax=<180 180 180>}
    ]
    rotationOffsets [<0 0 0>, <0 0 0>, <0 0 0>, <0 0 0>, <0 0 0>, <0 0 0>]
    trailBone {boneName=cape_07_l}
    trailRelaxationSpeed {editorCurveData {keys [<at 0 15 RCIM_Linear>,<at 1 6 RCIM_Linear>]}}
   }
   n22 = Trail controller - Bone: cape_07_r!(ComponentPose=@n21, Alpha=1, RelaxationSpeedScale:float=2, FakeVelocity:Vector="0,0,0") -> (Pose) {
    bActorSpaceFakeVel=true
    bShowTrailLocation=true
    bUsePlanarLimit=true
    chainLength=6
    debugLifeTime=0.1
    maxDeltaTime=0.0333
    relaxationSpeedScale=1.2
    alphaScaleBiasClamp {bInterpResult=true, interpSpeedDecreasing=25, interpSpeedIncreasing=25}
    baseJoint {boneName=backpack_root}
    rotationLimits [
     {limitMax=<180 180 180>}
     {limitMax=<180 180 180>}
     {limitMax=<180 180 180>}
     {limitMax=<180 180 180>}
     {limitMax=<180 180 180>}
     {limitMax=<180 180 180>}
    ]
    rotationOffsets [<0 0 0>, <0 0 0>, <0 0 0>, <0 0 0>, <0 0 0>, <0 0 0>]
    trailBone {boneName=cape_07_r}
    trailRelaxationSpeed {editorCurveData {keys [<at 0 15 RCIM_Linear>,<at 1 6 RCIM_Linear>]}}
   }
   Save cached pose 'dbnoCache'(Pose=Component To Local(ComponentPose=@n22)) {cacheName=dbnoCache}
   n23 = Anim Dynamics - Bone: cape_02_l(ComponentPose=Local To Component(LocalPose=Use cached pose 'constraintCache'()), Alpha=1, ExternalForce="0,0,0") {
    angularDampingOverride=1
    angularSpringConstant=10000
    bAngularSpring=true
    bOverrideAngularDamping=true
    bUseGravityOverride=true
    gravityScale=0.35
    simulationSpace=BoneRelative
    alphaBoolBlend {blendOption=Linear}
    boundBone {boneName=cape_02_l}
    componentAppliedLinearAccClamp=<3000 3000 3000>
    componentLinearAccScale=<2 2 2>
    componentLinearVelScale=<2 2 2>
    gravityOverride {z=-980}
    physicsBodyDefinitions [{
      boundBone {boneName=cape_02_l}
      boxExtents=<10 40 40>
      constraintSetup {angularLimitsMax=<0 0 15>, angularLimitsMin=<0 -40 -15>, angularTarget=<1 0 0>}
      localJointOffset=<5 0 0>
     }]
    relativeSpaceBone {boneName=backpack_root}
   }
   n24 = Anim Dynamics - Bone: cape_04_l(ComponentPose=@n23, Alpha=1, ExternalForce="0,0,0") {
    angularDampingOverride=1
    angularSpringConstant=10000
    bAngularSpring=true
    bOverrideAngularDamping=true
    bUseGravityOverride=true
    gravityScale=0.35
    simulationSpace=BoneRelative
    alphaBoolBlend {blendOption=Linear}
    boundBone {boneName=cape_04_l}
    componentAppliedLinearAccClamp=<3000 3000 3000>
    componentLinearAccScale=<2 2 2>
    componentLinearVelScale=<2 2 2>
    gravityOverride {z=-980}
    physicsBodyDefinitions [{
      boundBone {boneName=cape_04_l}
      boxExtents=<10 40 40>
      constraintSetup {angularLimitsMax=<0 0 15>, angularLimitsMin=<0 -40 -15>, angularTarget=<1 0 0>}
      localJointOffset=<5 0 0>
     }]
    relativeSpaceBone {boneName=backpack_root}
   }
   n25 = Anim Dynamics - Bone: cape_06_l(ComponentPose=@n24, Alpha=1, ExternalForce="0,0,0") {
    angularDampingOverride=1
    angularSpringConstant=10000
    bAngularSpring=true
    bOverrideAngularDamping=true
    bUseGravityOverride=true
    gravityScale=0.35
    simulationSpace=BoneRelative
    alphaBoolBlend {blendOption=Linear}
    boundBone {boneName=cape_06_l}
    componentAppliedLinearAccClamp=<3000 3000 3000>
    componentLinearAccScale=<2 2 2>
    componentLinearVelScale=<2 2 2>
    gravityOverride {z=-980}
    physicsBodyDefinitions [{
      boundBone {boneName=cape_06_l}
      boxExtents=<10 40 40>
      constraintSetup {angularLimitsMax=<0 5 15>, angularLimitsMin=<0 -40 -15>, angularTarget=<1 0 0>}
      localJointOffset=<5 0 0>
     }]
    relativeSpaceBone {boneName=backpack_root}
   }
   n26 = Anim Dynamics - Bone: cape_02_r(ComponentPose=@n25, Alpha=1, ExternalForce="0,0,0") {
    angularDampingOverride=1
    angularSpringConstant=10000
    bAngularSpring=true
    bOverrideAngularDamping=true
    bUseGravityOverride=true
    gravityScale=0.35
    simulationSpace=BoneRelative
    alphaBoolBlend {blendOption=Linear}
    boundBone {boneName=cape_02_r}
    componentAppliedLinearAccClamp=<3000 3000 3000>
    componentLinearAccScale=<2 2 2>
    componentLinearVelScale=<2 2 2>
    gravityOverride {z=-980}
    physicsBodyDefinitions [{
      boundBone {boneName=cape_02_r}
      boxExtents=<10 40 40>
      constraintSetup {angularLimitsMax=<0 0 15>, angularLimitsMin=<0 -40 -15>, angularTarget=<1 0 0>}
      localJointOffset=<5 0 0>
     }]
    relativeSpaceBone {boneName=backpack_root}
   }
   n27 = Anim Dynamics - Bone: cape_04_r(ComponentPose=@n26, Alpha=1, ExternalForce="0,0,0") {
    angularDampingOverride=1
    angularSpringConstant=10000
    bAngularSpring=true
    bOverrideAngularDamping=true
    bUseGravityOverride=true
    gravityScale=0.35
    simulationSpace=BoneRelative
    alphaBoolBlend {blendOption=Linear}
    boundBone {boneName=cape_04_r}
    componentAppliedLinearAccClamp=<3000 3000 3000>
    componentLinearAccScale=<2 2 2>
    componentLinearVelScale=<2 2 2>
    gravityOverride {z=-980}
    physicsBodyDefinitions [{
      boundBone {boneName=cape_04_r}
      boxExtents=<10 40 40>
      constraintSetup {angularLimitsMax=<0 0 15>, angularLimitsMin=<0 -40 -15>, angularTarget=<1 0 0>}
      localJointOffset=<5 0 0>
     }]
    relativeSpaceBone {boneName=backpack_root}
   }
   n28 = Anim Dynamics - Bone: cape_06_r(ComponentPose=@n27, Alpha=1, ExternalForce="0,0,0") {
    angularDampingOverride=1
    angularSpringConstant=10000
    bAngularSpring=true
    bOverrideAngularDamping=true
    bUseGravityOverride=true
    gravityScale=0.35
    simulationSpace=BoneRelative
    alphaBoolBlend {blendOption=Linear}
    boundBone {boneName=cape_06_r}
    componentAppliedLinearAccClamp=<3000 3000 3000>
    componentLinearAccScale=<2 2 2>
    componentLinearVelScale=<2 2 2>
    gravityOverride {z=-980}
    physicsBodyDefinitions [{
      boundBone {boneName=cape_06_r}
      boxExtents=<10 40 40>
      constraintSetup {angularLimitsMax=<0 5 15>, angularLimitsMin=<0 -40 -15>, angularTarget=<1 0 0>}
      localJointOffset=<5 0 0>
     }]
    relativeSpaceBone {boneName=backpack_root}
   }
   n29 = Anim Dynamics - Bone: cape_mid_01_l(ComponentPose=@n28, Alpha=1, ExternalForce="0,0,0") {
    angularDampingOverride=1
    angularSpringConstant=10000
    bAngularSpring=true
    bOverrideAngularDamping=true
    bUseGravityOverride=true
    gravityScale=0.35
    simulationSpace=BoneRelative
    alphaBoolBlend {blendOption=Linear}
    boundBone {boneName=cape_mid_01_l}
    componentAppliedLinearAccClamp=<3000 3000 3000>
    componentLinearAccScale=<2 2 2>
    componentLinearVelScale=<2 2 2>
    gravityOverride {z=-980}
    physicsBodyDefinitions [{
      boundBone {boneName=cape_mid_01_l}
      boxExtents=<10 40 40>
      constraintSetup {angularLimitsMax=<0 0 15>, angularLimitsMin=<0 -40 -15>, angularTarget=<1 0 0>}
      localJointOffset=<5 0 0>
     }]
    relativeSpaceBone {boneName=backpack_root}
   }
   n30 = Anim Dynamics - Bone: cape_mid_03_l(ComponentPose=@n29, Alpha=1, ExternalForce="0,0,0") {
    angularDampingOverride=1
    angularSpringConstant=10000
    bAngularSpring=true
    bOverrideAngularDamping=true
    bUseGravityOverride=true
    gravityScale=0.35
    simulationSpace=BoneRelative
    alphaBoolBlend {blendOption=Linear}
    boundBone {boneName=cape_mid_03_l}
    componentAppliedLinearAccClamp=<3000 3000 3000>
    componentLinearAccScale=<2 2 2>
    componentLinearVelScale=<2 2 2>
    gravityOverride {z=-980}
    physicsBodyDefinitions [{
      boundBone {boneName=cape_mid_03_l}
      boxExtents=<10 40 40>
      constraintSetup {angularLimitsMax=<0 0 15>, angularLimitsMin=<0 -40 -15>, angularTarget=<1 0 0>}
      localJointOffset=<5 0 0>
     }]
    relativeSpaceBone {boneName=backpack_root}
   }
   n31 = Anim Dynamics - Bone: cape_mid_05_l(ComponentPose=@n30, Alpha=1, ExternalForce="0,0,0") {
    angularDampingOverride=1
    angularSpringConstant=10000
    bAngularSpring=true
    bOverrideAngularDamping=true
    bUseGravityOverride=true
    gravityScale=0.35
    simulationSpace=BoneRelative
    alphaBoolBlend {blendOption=Linear}
    boundBone {boneName=cape_mid_05_l}
    componentAppliedLinearAccClamp=<3000 3000 3000>
    componentLinearAccScale=<2 2 2>
    componentLinearVelScale=<2 2 2>
    gravityOverride {z=-980}
    physicsBodyDefinitions [{
      boundBone {boneName=cape_mid_05_l}
      boxExtents=<10 40 40>
      constraintSetup {angularLimitsMax=<0 5 15>, angularLimitsMin=<0 -40 -15>, angularTarget=<1 0 0>}
      localJointOffset=<5 0 0>
     }]
    relativeSpaceBone {boneName=backpack_root}
   }
   n32 = Anim Dynamics - Bone: cape_mid_01_r(ComponentPose=@n31, Alpha=1, ExternalForce="0,0,0") {
    angularDampingOverride=1
    angularSpringConstant=10000
    bAngularSpring=true
    bOverrideAngularDamping=true
    bUseGravityOverride=true
    gravityScale=0.35
    simulationSpace=BoneRelative
    alphaBoolBlend {blendOption=Linear}
    boundBone {boneName=cape_mid_01_r}
    componentAppliedLinearAccClamp=<3000 3000 3000>
    componentLinearAccScale=<2 2 2>
    componentLinearVelScale=<2 2 2>
    gravityOverride {z=-980}
    physicsBodyDefinitions [{
      boundBone {boneName=cape_mid_01_r}
      boxExtents=<10 40 40>
      constraintSetup {angularLimitsMax=<0 0 15>, angularLimitsMin=<0 -40 -15>, angularTarget=<1 0 0>}
      localJointOffset=<5 0 0>
     }]
    relativeSpaceBone {boneName=backpack_root}
   }
   n33 = Anim Dynamics - Bone: cape_mid_03_r(ComponentPose=@n32, Alpha=1, ExternalForce="0,0,0") {
    angularDampingOverride=1
    angularSpringConstant=10000
    bAngularSpring=true
    bOverrideAngularDamping=true
    bUseGravityOverride=true
    gravityScale=0.35
    simulationSpace=BoneRelative
    alphaBoolBlend {blendOption=Linear}
    boundBone {boneName=cape_mid_03_r}
    componentAppliedLinearAccClamp=<3000 3000 3000>
    componentLinearAccScale=<2 2 2>
    componentLinearVelScale=<2 2 2>
    gravityOverride {z=-980}
    physicsBodyDefinitions [{
      boundBone {boneName=cape_mid_03_r}
      boxExtents=<10 40 40>
      constraintSetup {angularLimitsMax=<0 0 15>, angularLimitsMin=<0 -40 -15>, angularTarget=<1 0 0>}
      localJointOffset=<5 0 0>
     }]
    relativeSpaceBone {boneName=backpack_root}
   }
   n34 = Anim Dynamics - Bone: cape_mid_05_r(ComponentPose=@n33, Alpha=1, ExternalForce="0,0,0") {
    angularDampingOverride=1
    angularSpringConstant=10000
    bAngularSpring=true
    bOverrideAngularDamping=true
    bUseGravityOverride=true
    gravityScale=0.35
    simulationSpace=BoneRelative
    alphaBoolBlend {blendOption=Linear}
    boundBone {boneName=cape_mid_05_r}
    componentAppliedLinearAccClamp=<3000 3000 3000>
    componentLinearAccScale=<2 2 2>
    componentLinearVelScale=<2 2 2>
    gravityOverride {z=-980}
    physicsBodyDefinitions [{
      boundBone {boneName=cape_mid_05_r}
      boxExtents=<10 40 40>
      constraintSetup {angularLimitsMax=<0 5 15>, angularLimitsMin=<0 -40 -15>, angularTarget=<1 0 0>}
      localJointOffset=<5 0 0>
     }]
    relativeSpaceBone {boneName=backpack_root}
   }
   Save cached pose 'animDyn Cache'(Pose=Component To Local(ComponentPose=@n34)) {cacheName=animDyn Cache}
   n35 = RigidBody!(
    ComponentPose=Local To Component(LocalPose=Use cached pose 'Start_Cache'())
    Alpha
    OverrideWorldGravity:Vector="0,0,0"
    ExternalForce
    SimSpaceSettings:Sim Space Settings Structure=(WorldAlpha=1,VelocityScaleZ=1,DampingAlpha=0,MaxLinearVelocity=800,MaxAngularVelocity=8,MaxLinearAcceleration=4000,MaxAngularAcceleration=40,ExternalLinearDragV=(X=0,Y=0,Z=0),ExternalLinearVelocity=(X=0,Y=0,Z=0),ExternalAngularVelocity=(X=0,Y=0,Z=0))
   ) -> (Pose) {
    bClampLinearTranslationLimitToRefPose=true
    bOverrideWorldGravity=true
    overridePhysicsAsset=&/M_MED_WealthLamb_Fate_Cape_Dyn
    simulationSpace=BaseBoneSpace
    alphaBoolBlend {blendOption=Linear}
    baseBoneRef {boneName=spine_05}
    componentAppliedLinearAccClamp=<3000 3000 3000>
    componentLinearAccScale {z=0.5}
    componentLinearVelScale {z=0.5}
    overrideWorldGravity {z=-980}
    simSpaceSettings {dampingAlpha=0, maxAngularAcceleration=40, maxAngularVelocity=8, maxLinearAcceleration=4000, maxLinearVelocity=800, worldAlpha=1}
   }
   Save cached pose 'RbanCache'(Pose=Blend Poses by bool(
     BlendPose_0=Use cached pose 'Start_Cache'()
     BlendPose_1=Component To Local(ComponentPose=@n35)
     BlendTime_0=0.2
     BlendTime_1=0.5
     bActiveValue=Get bClothEnabled()
    ) {blendType=Linear}) {cacheName=RbanCache}
   n36 = "Constraint - cape_01_l (PPPPPPPP)"!(
    ComponentPose=Local To Component(LocalPose=Use cached pose 'Start_Cache'())
    Alpha=1
    ConstraintWeights_0=0.3884
    ConstraintWeights_1:float=0.2463
    ConstraintWeights_2:float=0.202
    ConstraintWeights_3:float=0.1145
    ConstraintWeights_4:float=0.0248
    ConstraintWeights_5:float=0.0137
    ConstraintWeights_6:float=0.0101
    ConstraintWeights_7:float=0.0002
   ) -> (Pose) {
    alphaBoolBlend {blendOption=Linear}
    boneToModify {boneName=cape_01_l}
    constraintSetup [
     {targetBone {boneName=clavicle_l}}
     {targetBone {boneName=spine_05}}
     {targetBone {boneName=deform_upperarm_up_l}}
     {targetBone {boneName=upperarm_twist_01_l}}
     {targetBone {boneName=deform_upperarm_bk_l}}
     {targetBone {boneName=upperarm_twist_02_l}}
     {targetBone {boneName=spine_04}}
     {targetBone {boneName=deform_upperarm_fr_l}}
    ]
    constraintWeights [1,1,1,1,1,1,1,1]
   }
   n37 = "Constraint - cape_mid_01_l_MSTR (PPPPPPPP)"!(
    ComponentPose=@n36
    Alpha=1
    ConstraintWeights_0=0.3841
    ConstraintWeights_1:float=0.243
    ConstraintWeights_2:float=0.103
    ConstraintWeights_3:float=0.0994
    ConstraintWeights_4:float=0.0669
    ConstraintWeights_5:float=0.0608
    ConstraintWeights_6:float=0.0386
    ConstraintWeights_7:float=0.0043
   ) -> (Pose) {
    alphaBoolBlend {blendOption=Linear}
    boneToModify {boneName=cape_mid_01_l_MSTR}
    constraintSetup [
     {targetBone {boneName=clavicle_l}}
     {targetBone {boneName=spine_05}}
     {targetBone {boneName=spine_04}}
     {targetBone {boneName=spine_03}}
     {targetBone {boneName=upperarm_twist_01_l}}
     {targetBone {boneName=deform_upperarm_up_l}}
     {targetBone {boneName=deform_upperarm_bk_l}}
     {targetBone {boneName=upperarm_twist_02_l}}
    ]
    constraintWeights [1,1,1,1,1,1,1,1]
   }
   n38 = "Constraint - cape_mid_01_r_MSTR (PPPPPPPP)"!(
    ComponentPose=@n37
    Alpha=1
    ConstraintWeights_0=0.3841
    ConstraintWeights_1:float=0.243
    ConstraintWeights_2:float=0.103
    ConstraintWeights_3:float=0.0994
    ConstraintWeights_4:float=0.0669
    ConstraintWeights_5:float=0.0608
    ConstraintWeights_6:float=0.0386
    ConstraintWeights_7:float=0.0043
   ) -> (Pose) {
    alphaBoolBlend {blendOption=Linear}
    boneToModify {boneName=cape_mid_01_r_MSTR}
    constraintSetup [
     {targetBone {boneName=clavicle_r}}
     {targetBone {boneName=spine_05}}
     {targetBone {boneName=spine_04}}
     {targetBone {boneName=spine_03}}
     {targetBone {boneName=upperarm_twist_01_r}}
     {targetBone {boneName=deform_upperarm_up_r}}
     {targetBone {boneName=deform_upperarm_bk_r}}
     {targetBone {boneName=upperarm_twist_02_r}}
    ]
    constraintWeights [1,1,1,1,1,1,1,1]
   }
   n39 = "Constraint - cape_01_r (PPPPPPPP)"!(
    ComponentPose=@n38
    Alpha=1
    ConstraintWeights_0=0.3884
    ConstraintWeights_1:float=0.2463
    ConstraintWeights_2:float=0.202
    ConstraintWeights_3:float=0.1145
    ConstraintWeights_4:float=0.0248
    ConstraintWeights_5:float=0.0137
    ConstraintWeights_6:float=0.0101
    ConstraintWeights_7:float=0.0002
   ) -> (Pose) {
    alphaBoolBlend {blendOption=Linear}
    boneToModify {boneName=cape_01_r}
    constraintSetup [
     {targetBone {boneName=clavicle_r}}
     {targetBone {boneName=spine_05}}
     {targetBone {boneName=deform_upperarm_up_r}}
     {targetBone {boneName=upperarm_twist_01_r}}
     {targetBone {boneName=deform_upperarm_bk_r}}
     {targetBone {boneName=upperarm_twist_02_r}}
     {targetBone {boneName=spine_04}}
     {targetBone {boneName=deform_upperarm_fr_r}}
    ]
    constraintWeights [1,1,1,1,1,1,1,1]
   }
   n40 = "Constraint - cape_02_l (P)"(ComponentPose=@n39, Alpha=1, ConstraintWeights_0=1) {
    alphaBoolBlend {blendOption=Linear}
    boneToModify {boneName=cape_02_l}
    constraintSetup [{targetBone {boneName=cape_mid_01_l}}]
    constraintWeights [1]
   }
   n41 = "Constraint - cape_02_r (P)"(ComponentPose=@n40, Alpha=1, ConstraintWeights_0=1) {
    alphaBoolBlend {blendOption=Linear}
    boneToModify {boneName=cape_02_r}
    constraintSetup [{targetBone {boneName=cape_mid_01_r}}]
    constraintWeights [1]
   }
   n42 = "Transform (Modify) Bone - Bone: backpack_root"(ComponentPose=@n41, Alpha=1, Translation="-0.000,-1.776,0.567", Rotation="0,0,0", Scale="1,1,1") {
    rotationMode=BMM_Additive
    rotationSpace=BCS_BoneSpace
    translationMode=BMM_Additive
    translationSpace=BCS_BoneSpace
    alphaBoolBlend {blendOption=Linear}
    boneToModify {boneName=backpack_root}
    translation {y=-1.776, z=0.567}
   }
   Save cached pose 'constraintCache'(Pose=Component To Local(ComponentPose=@n42)) {cacheName=constraintCache}
   Use cached pose 'Start_Cache'()
   Use cached pose 'Start_Cache'()
  ]}
 EventGraph {nodes [
   n1 = "Cloth Controller Function (Parameter Asset) (Nativized)"!(
    execute=@n11.then
    CharacterPartAnimInstance:Custom Character Part Anim Instance Object Reference=Self-Reference()
    ClothingAssetName:Name=Base_Cape_01
    ClothCoreSettings:"Cloth Core Settings Structure (by ref)"=Get Cloth Core Settings_Native() -> (Cloth Core Settings_Native:Cloth Core Settings Structure)
    WindGustLocalSettings:"Wind Gust Settings Structure (by ref)"=Get Wind Gust Local Settings Native() -> (Wind Gust Local Settings Native:Wind Gust Settings Structure)
    ClothUpAxis:Vector="1,0,0"
    LinVelScaleFwdBwdClamp:float=0
    LinVelScaleUpDwnClamp:float=0.7
    DeltaTime:float=Get DeltaTimeLocal() -> (DeltaTimeLocal:double)
    GameTimeInSeconds:float=GetGameTimeInSeconds()
    StateLogicLayerName:Name=none
    Parameters:Fort Character Dynamics Parameters Object Reference=&/M_MED_WealthLamb_Fate_Cape_Parameters
   ) -> (
    then
    OutCurrentState:Name=None
    OutClothParameters:Cloth Parameters Structure
    OutClothCoreSettings:Cloth Core Settings Structure
    OutWindGustSettings:Wind Gust Parameters Structure
    OutWindGustLocalSettings:Wind Gust Settings Structure
    OutWindVelocity:Vector
    OutLinearVelVector:Vector
   )
   n2 = Get Skel_Mesh()
   n3 = DrawVectorFromJointFunction!(
    execute
    SkelMeshComponent:Scene Component Object Reference=@n2
    JointName:Name=spine_05
    LineEnd:Vector=Break Cloth Parameters!(ClothParameters:"Cloth Parameters Structure (by ref)"=@n1.OutClothParameters) -> (
     LinearVelocityScale:Vector=~
     AngularVelocityScale:float=~
     FictitiousAngularScale:float=~
     GravityOverride:Vector=~
     DragCoefficient:Vector 2D Structure
     LiftCoefficient:Vector 2D Structure
     OuterDrag:Vector 2D Structure
     OuterLift:Vector 2D Structure
     MaxLinearVelocity:Vector=~
     MaxLinearAcceleration:Vector=~
     MaxAngularVelocity:float=~
     MaxAngularAcceleration:float=~
     bEnableLinearVelocityClamping:bool=~
     bEnableLinearAccelerationClamping:bool=~
     bEnableAngularVelocityClamping:bool=~
     bEnableAngularAccelerationClamping:bool=~
     JointName:Name
    ) {is=title K2Node_BreakStruct}.GravityOverride
    JointOffset:Vector="0,0,30"
    LineColor:Linear Color Structure=(R=0.0192,G=0,B=1,A=1)
    Thickness:double=1
    Foreground:EDrawDebugSceneDepthPriorityGroup Enum=Foreground
   ) -> (then)
   Debug Player Velocities Function!(
    execute=[@n3.then, @n7.then]
    CharacterPartAnimInstance:Custom Character Part Anim Instance Object Reference=Self-Reference()
    SkelMeshComponent:Scene Component Object Reference
    DeltaTime:double=0
    DebugSpace:DebugDrawSpace Enum=NewEnumerator1
    BoneName:Name=root
    DebugDrawOffset:Vector="0,0,105"
    AngularDebugDrawOffset:Vector="0,0,0"
    DrawDebug:bool=true
    Debug Scalar:double=50
    Foreground:EDrawDebugSceneDepthPriorityGroup Enum=Foreground
    TextLocationOffset:Vector="20,20,20"
   ) -> (
    then
    Component VelocityX Vector:Vector
    Component VelocityY Vector:Vector
    Component VelocityZ Vector:Vector
    RotationRateDegrees:float
    BoneY Axis1:double
    BoneX Axis1:double
    BoneZ Axis1:double
    BoneLocationPassed1:Vector
    ReturnValue3:Vector
    ReturnValue4:Vector
    ReturnValue5:Vector
    ComponentRightVectorDot1:double
    ComponentFwdVectorDot1:double
    ComponentUpVectorDot1:double
    PawnVelocity1:Vector
   )
   n4 = "Gravity Override State Function (Parameter Asset) (Nativized)"(
    execute=@n11.else
    CharacterPartAnimInstance=Self-Reference()
    GameTimeInSeconds=GetGameTimeInSeconds()
    Parameters=&M_MED_WealthLamb_Fate_Cape_Parameters
   )
   > n5 = Set GravityOverride(GravityOverride=@n4)
   n6 = DrawVectorFromJointFunction(
    SkelMeshComponent=@n2
    JointName=head
    LineEnd=@n5
    JointOffset="0, 0, 0"
    LineColor=(R=0.0192,G=0,B=1,A=1)
    Thickness=2
    Foreground=Foreground
   )
   n7 = DrawJointDebugFunction!(
    execute=[@n6.then, @n8.then]
    SkelMeshComponent:Scene Component Object Reference=@n2
    JointOffset:Vector="0, 0, 0"
    JointName:Array of Names=Make Array!("[0]":Name=spine_05, "[1]":Name=root) -> (Array:Array of Names)
    Scale:double=10
    Duration:double=0
    Thickness:double=0.7
    DebugDraw:bool=true
    Foreground:EDrawDebugSceneDepthPriorityGroup Enum=Foreground
   ) -> (then)
   n8 = DrawVectorFromJointFunction(
    SkelMeshComponent=@n2
    JointName=spine_05
    LineEnd=@n1.OutWindVelocity
    JointOffset="0,0,30"
    LineColor=(R=1,G=0,B=0,A=1)
    Thickness=1
    Foreground=Foreground
   )
   n9 = Event BlueprintUpdateAnimation()
   > Set DeltaTimeLocal(DeltaTimeLocal:double=@n9.DeltaTimeX) -> (Output_Get:double)
   > n10 = Sequence!(execute) -> (then_0, then_1, then_2:Exec, then_3:Exec)
   > n11 = Branch(Condition=Get bClothEnabled())
   n12 = PlayerMovement(execute=@n10.then_1, AnimInstance=Self-Reference(), XDotProductCoefficient=1, YDotProductCoefficient=1, ZDotProductCoefficient=1)
   n13 = MapRangeClamped(Value=@n12.YMovement, InRangeA=0, InRangeB=1)
   n14 = "Absolute (Float)"(A=@n12)
   > Set BwdMovement(BwdMovement:double=MapRangeClamped(Value=float + float(A=@n14, B=@n13) {is=title K2Node_CommutativeAssociativeBinaryOperator}, InRangeA=0, InRangeB=1.4)) -> (Output_Get:double)
   n15 = Branch(execute=@n10.then_2, Condition=Get bIsSwimSprinting() -> (bIsSwimSprinting:bool))
   > Set TrailControllerRot(TrailControllerRot:Rotator="15,0, 0") -> (Output_Get:Rotator)
   n16 = Branch(execute=@n15.else, Condition=Get bIsSkyDiving())
   > n17 = Branch(Condition=Get bIsParachuteOpened())
   > Set TrailControllerRot(TrailControllerRot="45,0, 0")
   n18 = Branch(execute=@n17.else, Condition="float > float"(A=Get PawnVelocityZ(), B=50) {is=title K2Node_CallFunction})
   > Set BwdMovement(BwdMovement=float * float(A=Get BwdMovement(), B=0))
   > Set TrailControllerRot(TrailControllerRot="45,0, 0")
   n19 = Branch(execute=@n18.else, Condition=Get bIsSkydiveDiveMode())
   > Set TrailControllerRot(TrailControllerRot="50,0,0")
   Set TrailControllerRot(execute=@n19.else, TrailControllerRot="60,0, 0")
   Set TrailControllerRot(execute=@n16.else, TrailControllerRot="15,0, 0")
   > Set BwdMovement(BwdMovement=float * float(A=Get BwdMovement(), B=1))
   K2Node_Composite_0 = IsicleLarva_Event CCF!(then_10:Exec=@n10.then_3) {is=title K2Node_Composite}
   Event BlueprintInitializeAnimation()
   > Set Skel_Mesh(Skel_Mesh=GetOwnerSkeletalMeshComponent())
   Event OnBeginSkydiving()
   > Set TrailControllerRot(TrailControllerRot="-65,0,0")
   Event OnEndSkydiving()
   > Set TrailControllerRot(TrailControllerRot="-10,0,0")
   Set RBAN Settings Out(execute=@n5.then, RBAN Settings Out=@n4.OutRigidBodyAnimNodeParameters)
   Set Wind Gust Local Settings Out(execute=@n1.then, Wind Gust Local Settings Out:Wind Gust Settings Structure=@n1.OutWindGustLocalSettings) -> (Output_Get:Wind Gust Settings Structure)
  ]}
 IsicleLarva_Event CCF {
  path=EventGraph.K2Node_Composite_0.IsicleLarva_Event CCF
  nodes [
   n1 = Break Cloth Setting Struct!(ClothSettingStruct:"Cloth Setting Struct Structure (by ref)"=Get Cloth Settings FINAL() -> (Cloth Settings FINAL:Cloth Setting Struct Structure)) -> (
    LinearVelocityScale_7_#1:Vector=~
    AngularVelocityScale_6_#2:float=~
    FictitiousAngularScale_10_#3:float=~
    GravityOverride_15_#4:Vector=~
    DragCoefficient_23_#5:Vector 2D Structure
    LiftCoefficient_24_#6:Vector 2D Structure
    JointName_27_#7:Name
   ) {is=title K2Node_BreakStruct}
   n2 = Get PartSkelMeshComp() -> (PartSkelMeshComp:Skeletal Mesh Component Object Reference)
   DrawVectorFromJointFunction(
    SkelMeshComponent=@n2
    JointName=backpack_root
    LineEnd=@n1.GravityOverride_15_#4
    JointOffset="0, 0, 0"
    LineColor=(R=1,G=0,B=0,A=1)
    Thickness=2
    Foreground=Foreground
   )
   > DrawVectorFromJointFunction(
    SkelMeshComponent=@n2
    JointName=backpack_root
    LineEnd=Get Wind Velocity FINAL() -> (Wind Velocity FINAL:Vector)
    JointOffset="0, 0, 0"
    LineColor=(R=0.9964,G=1,B=0.0017,A=1)
    Thickness=1.5
    Foreground=Foreground
   )
   > Debug Player Velocities Function(
    CharacterPartAnimInstance=Self-Reference()
    DeltaTime=0
    DebugSpace=NewEnumerator1
    BoneName=root
    DebugDrawOffset="0,0,0"
    AngularDebugDrawOffset="0,0,0"
    DrawDebug=true
    Debug Scalar=200
    Foreground=Foreground
    TextLocationOffset="20,20,20"
   )
   > DrawJointDebugFunction(
    SkelMeshComponent=@n2
    JointOffset="0,0,10"
    JointName=Make Array("[0]"=pelvis, "[1]"=root)
    Scale=10
    Duration=0
    Thickness=0.7
    DebugDraw=true
    Foreground=Foreground
   )
   n3 = Get As Fort Player Pawn() -> (As Fort Player Pawn:Fort Player Pawn Object Reference)
   n4 = Cast To FortAnimInstance!(execute, Object=GetAnimInstance(self=Get Mesh(self=@n3))) -> (then, CastFailed, AsFort Anim Instance:Fort Anim Instance Object Reference) {is=title K2Node_DynamicCast}
   > RequestFullBodyPBALayer!(execute, InAnimInstance:Fort Anim Instance Object Reference=@n4, ReasonForRequest:Name=IcicleLarva) -> (then)
   > SetProperty!(
    execute
    InAnimInstance:Fort Anim Instance Object Reference=@n4
    InPropertyName:Name=IcicleLarva_Flying
    InValue:bool=Get IcicleLarva_Flying() -> (IcicleLarva_Flying:bool)
   ) -> (then)
   > SetProperty(InAnimInstance=@n4, InPropertyName=IcicleLarva_Hover, InValue=Get IcicleLarva_Hover() -> (IcicleLarva_Hover:bool))
   > SetProperty(InAnimInstance=@n4, InPropertyName=IcicleLarva_SprintFly, InValue=Get IcicleLarva_SprintFly() -> (IcicleLarva_SprintFly:bool))
   > SetProperty(InAnimInstance=@n4, InPropertyName=IcicleLarva_AirCharge, InValue=Get IcicleLarva_AirCharge() -> (IcicleLarva_AirCharge:bool))
   Inputs!() -> (then_10:Exec)
   > n5 = Sequence!(execute) -> (then_0, then_1)
   > Set PartSkelMeshComp(PartSkelMeshComp:Skeletal Mesh Component Object Reference=GetPartSkeletalMeshComponent()) -> (Output_Get:Skeletal Mesh Component Object Reference)
   > Set OwnerSkelMeshComp(OwnerSkelMeshComp:Skeletal Mesh Component Object Reference=GetOwnerSkeletalMeshComponent()) -> (Output_Get:Skeletal Mesh Component Object Reference)
   n6 = GetClothingSimulationInteractor!(self:Skeletal Mesh Component Object Reference=Get PartSkelMeshComp(), ClothingAssetName:Name=None) -> (ReturnValue:Clothing Simulation Interactor Object Reference)
   Is Valid(exec=@n5.then_1, InputObject=@n6)
   n7 = MakeLiteralString(Value=Base_Cape_01)
   > n8 = Cast To ChaosClothingInteractor!(
    execute
    Object=GetClothingInteractor!(self:Clothing Simulation Interactor Object Reference=@n6, ClothingAssetName:Name=String To Name(InString=@n7)) -> (ReturnValue:Clothing Interactor Object Reference)
   ) -> (then, CastFailed, AsChaos Clothing Interactor:Chaos Clothing Interactor Object Reference) {is=title K2Node_DynamicCast}
   > Set ChaosClothingInteractor(ChaosClothingInteractor:Chaos Clothing Interactor Object Reference=@n8) -> (Output_Get:Chaos Clothing Interactor Object Reference)
   > n9 = Cast To FortPlayerPawn(Object=GetOwningActor())
   > Set As Fort Player Pawn(As Fort Player Pawn:Fort Player Pawn Object Reference=@n9) -> (Output_Get:Fort Player Pawn Object Reference)
   > n10 = Sequence!(execute) -> (then_0, then_1, then_2:Exec, then_3:Exec, then_4:Exec, then_5:Exec, then_6:Exec, then_7:Exec, then_8:Exec, then_9:Exec)
   n11 = Branch(execute=@n10.then_8, Condition=Get IcicleLarva_Ascend() -> (IcicleLarva_Ascend:bool))
   > Set CurrentState(CurrentState:String=Ascend) -> (Output_Get:String)
   > Set Cloth Core Settings FINAL(Cloth Core Settings FINAL:Cloth Core Setting Struct Structure=Get Cloth Core Settings Ascend() -> (Cloth Core Settings Ascend:Cloth Core Setting Struct Structure)) -> (Output_Get:Cloth Core Setting Struct Structure)
   n12 = Break Cloth Setting Struct(ClothSettingStruct=Get Cloth Settings Ascend() -> (Cloth Settings Ascend:Cloth Setting Struct Structure)) -> (
    LinearVelocityScale_7_#1=~
    AngularVelocityScale_6_#2=~
    FictitiousAngularScale_10_#3=~
    GravityOverride_15_#4=~
   )
   > n13 = AlignVectorWithJoint!(
    execute
    Vector:Vector=@n12.GravityOverride_15_#4
    MeshComp:Mesh Component Object Reference=Get OwnerSkelMeshComp() -> (OwnerSkelMeshComp:Skeletal Mesh Component Object Reference)
    PartAnimInstance:Custom Character Part Anim Instance Object Reference
    JointName:Name=@n12.JointName_27_#7
   ) -> (then, ReturnValue:Vector)
   > Set Cloth Settings FINAL(Cloth Settings FINAL:Cloth Setting Struct Structure=Make Cloth Setting Struct!(
     LinearVelocityScale_7_#1:Vector=@n12
     AngularVelocityScale_6_#2:float=@n12.AngularVelocityScale_6_#2
     FictitiousAngularScale_10_#3:float=@n12.FictitiousAngularScale_10_#3
     GravityOverride_15_#4:Vector=@n13
     DragCoefficient_23_#5:Vector 2D Structure=@n12.DragCoefficient_23_#5
     LiftCoefficient_24_#6:Vector 2D Structure=@n12.LiftCoefficient_24_#6
     JointName_27_#7:Name=@n12.JointName_27_#7
    ) -> (ClothSettingStruct:Cloth Setting Struct Structure) {is=title K2Node_MakeStruct}) -> (Output_Get:Cloth Setting Struct Structure)
   > Set Max Linear Velocity FINAL(Max Linear Velocity FINAL:Vector="800,800,800") -> (Output_Get:Vector)
   > Set Max Linear Acceleration FINAL(Max Linear Acceleration FINAL:Vector="1800,1800,1800") -> (Output_Get:Vector)
   > Set Max Angular Velocity FINAL(Max Angular Velocity FINAL:float=2) -> (Output_Get:float)
   > Set Max Angular Acceleration FINAL(Max Angular Acceleration FINAL:float=10) -> (Output_Get:float)
   > n14 = Random Vector Function!(
    execute
    CharacterPartAnimInstance:Custom Character Part Anim Instance Object Reference
    SkelMeshComponent:Mesh Component Object Reference=Get OwnerSkelMeshComp()
    JointName:Name=@n12.JointName_27_#7
    GameTimeInSeconds:double=GetGameTimeInSeconds()
    TimeOffset:double=0
    PerlinScale:double=1
    Frequency:Vector="3,2,1"
    Amplitude:Vector="70,130,150"
    Offset:Vector="0,-175,-400"
   ) -> (then, VectorOut:Vector)
   > n15 = Set Wind Velocity FINAL(Wind Velocity FINAL:Vector=@n14) -> (Output_Get:Vector)
   n16 = Get ChaosClothingInteractor() -> (ChaosClothingInteractor:Chaos Clothing Interactor Object Reference)
   SetVelocityScale!(
    execute=[@n37.then, @n23.then, @n47.then, @n52.then, @n15.then, @n42.then, @n31.then]
    self:Chaos Clothing Interactor Object Reference=@n16
    LinearVelocityScale:Vector=@n1
    AngularVelocityScale:float=@n1.AngularVelocityScale_6_#2
    FictitiousAngularScale:float=@n1.FictitiousAngularScale_10_#3
   ) -> (then)
   > SetVelocityClamps!(
    execute
    self:Chaos Clothing Interactor Object Reference=@n16
    bEnableLinearVelocityClamping:bool=true
    MaxLinearVelocity:Vector=Get Max Linear Velocity FINAL() -> (Max Linear Velocity FINAL:Vector)
    bEnableLinearAccelerationClamping:bool=true
    MaxLinearAcceleration:Vector=Get Max Linear Acceleration FINAL() -> (Max Linear Acceleration FINAL:Vector)
    bEnableAngularVelocityClamping:bool=true
    MaxAngularVelocity:float=Get Max Angular Velocity FINAL() -> (Max Angular Velocity FINAL:float)
    bEnableAngularAccelerationClamping:bool=true
    MaxAngularAcceleration:float=Get Max Angular Acceleration FINAL() -> (Max Angular Acceleration FINAL:float)
   ) -> (then)
   > SetGravity!(
    execute
    self:Chaos Clothing Interactor Object Reference=@n16
    GravityScale:float=1
    bIsGravityOverridden:bool=true
    GravityOverride:Vector=@n1.GravityOverride_15_#4
   ) -> (then)
   > SetWind!(
    execute
    self:Chaos Clothing Interactor Object Reference=@n16
    Drag:Vector 2D Structure=@n1.DragCoefficient_23_#5
    Lift:Vector 2D Structure=@n1.LiftCoefficient_24_#6
    AirDensity:float=0
    WindVelocity:Vector=Get Wind Velocity FINAL()
    OuterDrag:Vector 2D Structure=(X=0.07,Y=0.5)
    OuterLift:Vector 2D Structure=(X=0.07,Y=0.5)
   ) -> (then)
   n17 = Break Cloth Core Setting Struct!(ClothCoreSettingStruct:"Cloth Core Setting Struct Structure (by ref)"=Get Cloth Core Settings FINAL() -> (Cloth Core Settings FINAL:Cloth Core Setting Struct Structure)) -> (
    EdgeStiffness_28_#8:Vector 2D Structure
    BendingStiffness_29_#9:Vector 2D Structure
    AreaStiffness_30_#10:Vector 2D Structure
    TetherStiffness_24_#11:Vector 2D Structure
    CollisionThickness_14_#12:float=~
    FrictionCoefficient_15_#13:float=~
    SelfCollisionThickness_16_#14:float=~
    DampingCoefficient_17_#15:float=~
    LocalDampingCoefficient_33_#16:float=~
    AnimDriveStiffness_27_#17:Vector 2D Structure
    AnimDriveDamping_26_#18:Vector 2D Structure
    NumberIterations_21_#19:int
    NumberSubsteps_23_#20:int
   ) {is=title K2Node_BreakStruct}
   n18 = Get ChaosClothingInteractor()
   > SetMaterial!(
    execute
    self:Chaos Clothing Interactor Object Reference=@n18
    EdgeStiffness:Vector 2D Structure=@n17
    BendingStiffness:Vector 2D Structure=@n17.BendingStiffness_29_#9
    AreaStiffness:Vector 2D Structure=@n17.AreaStiffness_30_#10
   ) -> (then)
   > SetLongRangeAttachment!(
    execute
    self:Chaos Clothing Interactor Object Reference=@n18
    TetherStiffness:Vector 2D Structure=@n17.TetherStiffness_24_#11
    TetherScale:Vector 2D Structure=(X=1.05,Y=1.05)
   ) -> (then)
   > SetCollision!(
    execute
    self:Chaos Clothing Interactor Object Reference=@n18
    CollisionThickness:float=@n17.CollisionThickness_14_#12
    FrictionCoefficient:float=@n17.FrictionCoefficient_15_#13
    bUseCCD:bool=false
    SelfCollisionThickness:float=@n17.SelfCollisionThickness_16_#14
   ) -> (then)
   > SetDamping!(
    execute
    self:Chaos Clothing Interactor Object Reference=@n18
    DampingCoefficient:float=@n17.DampingCoefficient_17_#15
    LocalDampingCoefficient:float=@n17.LocalDampingCoefficient_33_#16
   ) -> (then)
   > SetAnimDrive!(
    execute
    self:Chaos Clothing Interactor Object Reference=@n18
    AnimDriveStiffness:Vector 2D Structure=@n17.AnimDriveStiffness_27_#17
    AnimDriveDamping:Vector 2D Structure=@n17.AnimDriveDamping_26_#18
   ) -> (then)
   n19 = Branch(execute=@n11.else, Condition=Get IcicleLarva_HeatVision() -> (IcicleLarva_HeatVision:bool))
   > Set CurrentState(CurrentState=Heat Vision)
   > Set Cloth Core Settings FINAL(Cloth Core Settings FINAL=Get Cloth Core Settings HeatVision() -> (Cloth Core Settings HeatVision:Cloth Core Setting Struct Structure))
   n20 = Break Cloth Setting Struct(ClothSettingStruct=Get Cloth Setting HeatVision() -> (Cloth Setting HeatVision:Cloth Setting Struct Structure)) -> (
    LinearVelocityScale_7_#1=~
    AngularVelocityScale_6_#2=~
    FictitiousAngularScale_10_#3=~
    GravityOverride_15_#4=~
   )
   > n21 = AlignVectorWithJoint(
    Vector=@n20.GravityOverride_15_#4
    MeshComp=Get OwnerSkelMeshComp()
    JointName=@n20.JointName_27_#7
   )
   > Set Cloth Settings FINAL(Cloth Settings FINAL=Make Cloth Setting Struct(
     LinearVelocityScale_7_#1=@n20
     AngularVelocityScale_6_#2=@n20.AngularVelocityScale_6_#2
     FictitiousAngularScale_10_#3=@n20.FictitiousAngularScale_10_#3
     GravityOverride_15_#4=@n21
     DragCoefficient_23_#5=@n20.DragCoefficient_23_#5
     LiftCoefficient_24_#6=@n20.LiftCoefficient_24_#6
     JointName_27_#7=@n20.JointName_27_#7
    ))
   > Set Max Linear Velocity FINAL(Max Linear Velocity FINAL="800,800,800")
   > Set Max Linear Acceleration FINAL(Max Linear Acceleration FINAL="1800,1800,1800")
   > Set Max Angular Velocity FINAL(Max Angular Velocity FINAL=2)
   > Set Max Angular Acceleration FINAL(Max Angular Acceleration FINAL=10)
   > n22 = Random Vector Function(
    SkelMeshComponent=Get OwnerSkelMeshComp()
    JointName=@n20.JointName_27_#7
    GameTimeInSeconds=GetGameTimeInSeconds()
    TimeOffset=0
    PerlinScale=1
    Frequency="3,2,1"
    Amplitude="70,130,150"
    Offset="0,-1000,-400"
   )
   > n23 = Set Wind Velocity FINAL(Wind Velocity FINAL=@n22)
   n24 = Branch(execute=@n19.else, Condition=Get IcicleLarva_Hover())
   > Set CurrentState(CurrentState=Hover)
   > Set Cloth Core Settings FINAL(Cloth Core Settings FINAL=Get Cloth Core Settings Hover() -> (Cloth Core Settings Hover:Cloth Core Setting Struct Structure))
   n25 = Break Cloth Setting Struct(ClothSettingStruct=Get Cloth Settings Hover() -> (Cloth Settings Hover:Cloth Setting Struct Structure)) -> (
    LinearVelocityScale_7_#1=~
    AngularVelocityScale_6_#2=~
    FictitiousAngularScale_10_#3=~
    GravityOverride_15_#4=~
   )
   > n26 = AlignVectorWithJoint(
    Vector=@n25.GravityOverride_15_#4
    MeshComp=Get OwnerSkelMeshComp()
    JointName=@n25.JointName_27_#7
   )
   > Set Cloth Settings FINAL(Cloth Settings FINAL=Make Cloth Setting Struct(
     LinearVelocityScale_7_#1=@n25
     AngularVelocityScale_6_#2=@n25.AngularVelocityScale_6_#2
     FictitiousAngularScale_10_#3=@n25.FictitiousAngularScale_10_#3
     GravityOverride_15_#4=@n26
     DragCoefficient_23_#5=@n25.DragCoefficient_23_#5
     LiftCoefficient_24_#6=@n25.LiftCoefficient_24_#6
     JointName_27_#7=@n25.JointName_27_#7
    ))
   n27 = Get PawnVelocityZ()
   > n28 = Branch(Condition=OR Boolean(A="float < float"(A=@n27, B=-100), B="float > float"(A=@n27, B=100) {is=title K2Node_PromotableOperator}))
   > Set Max Linear Velocity FINAL(Max Linear Velocity FINAL="100,500,500")
   > n29 = Set Max Linear Acceleration FINAL(Max Linear Acceleration FINAL="500,1500,1500")
   Set Max Angular Velocity FINAL(execute=[@n29.then, @n32.then], Max Angular Velocity FINAL=2)
   > Set Max Angular Acceleration FINAL(Max Angular Acceleration FINAL=10)
   > n30 = Random Vector Function(
    SkelMeshComponent=Get OwnerSkelMeshComp()
    JointName=@n25.JointName_27_#7
    GameTimeInSeconds=GetGameTimeInSeconds()
    TimeOffset=0
    PerlinScale=1
    Frequency="3,2,1"
    Amplitude="70,130,150"
    Offset="0,-175,-400"
   )
   > n31 = Set Wind Velocity FINAL(Wind Velocity FINAL=@n30)
   Set Max Linear Velocity FINAL(execute=@n28.else, Max Linear Velocity FINAL="800,800,800")
   > n32 = Set Max Linear Acceleration FINAL(Max Linear Acceleration FINAL="1800,1800,1800")
   n33 = Branch(execute=@n24.else, Condition=Get IcicleLarva_Flying())
   > Set CurrentState(CurrentState=Flying)
   > Set Cloth Core Settings FINAL(Cloth Core Settings FINAL=Get Cloth Core Settings Flying() -> (Cloth Core Settings Flying:Cloth Core Setting Struct Structure))
   n34 = Break Cloth Setting Struct(ClothSettingStruct=Get Cloth Settings Flying() -> (Cloth Settings Flying:Cloth Setting Struct Structure)) -> (
    LinearVelocityScale_7_#1=~
    AngularVelocityScale_6_#2=~
    FictitiousAngularScale_10_#3=~
    GravityOverride_15_#4=~
   )
   > n35 = AlignVectorWithJoint(
    Vector=@n34.GravityOverride_15_#4
    MeshComp=Get OwnerSkelMeshComp()
    JointName=@n34.JointName_27_#7
   )
   > Set Cloth Settings FINAL(Cloth Settings FINAL=Make Cloth Setting Struct(
     LinearVelocityScale_7_#1=@n34
     AngularVelocityScale_6_#2=@n34.AngularVelocityScale_6_#2
     FictitiousAngularScale_10_#3=@n34.FictitiousAngularScale_10_#3
     GravityOverride_15_#4=@n35
     DragCoefficient_23_#5=@n34.DragCoefficient_23_#5
     LiftCoefficient_24_#6=@n34.LiftCoefficient_24_#6
     JointName_27_#7=@n34.JointName_27_#7
    ))
   > Set Max Linear Velocity FINAL(Max Linear Velocity FINAL="800,800,800")
   > Set Max Linear Acceleration FINAL(Max Linear Acceleration FINAL="2000,2000,2000")
   > Set Max Angular Velocity FINAL(Max Angular Velocity FINAL=2)
   > Set Max Angular Acceleration FINAL(Max Angular Acceleration FINAL=10)
   > n36 = Random Vector Function(
    SkelMeshComponent=Get OwnerSkelMeshComp()
    JointName=@n34.JointName_27_#7
    GameTimeInSeconds=GetGameTimeInSeconds()
    TimeOffset=0
    PerlinScale=RandomFloatInRange(Min=1, Max=3)
    Frequency="5,7,12"
    Amplitude="50,70,45"
    Offset="-500,-50,0"
   )
   > n37 = Set Wind Velocity FINAL(Wind Velocity FINAL=@n36)
   n38 = Branch(execute=@n33.else, Condition=Get IcicleLarva_SprintFly())
   > Set CurrentState(CurrentState=Sprint Fly)
   > Set Cloth Core Settings FINAL(Cloth Core Settings FINAL=Get Cloth Core Settings SprintFly() -> (Cloth Core Settings SprintFly:Cloth Core Setting Struct Structure))
   n39 = Break Cloth Setting Struct(ClothSettingStruct=Get Cloth Settings SprintFly() -> (Cloth Settings SprintFly:Cloth Setting Struct Structure)) -> (
    LinearVelocityScale_7_#1=~
    AngularVelocityScale_6_#2=~
    FictitiousAngularScale_10_#3=~
    GravityOverride_15_#4=~
   )
   > n40 = AlignVectorWithJoint(
    Vector=@n39.GravityOverride_15_#4
    MeshComp=Get OwnerSkelMeshComp()
    JointName=@n39.JointName_27_#7
   )
   > Set Cloth Settings FINAL(Cloth Settings FINAL=Make Cloth Setting Struct(
     LinearVelocityScale_7_#1=@n39
     AngularVelocityScale_6_#2=@n39.AngularVelocityScale_6_#2
     FictitiousAngularScale_10_#3=@n39.FictitiousAngularScale_10_#3
     GravityOverride_15_#4=@n40
     DragCoefficient_23_#5=@n39.DragCoefficient_23_#5
     LiftCoefficient_24_#6=@n39.LiftCoefficient_24_#6
     JointName_27_#7=@n39.JointName_27_#7
    ))
   > Set Max Linear Velocity FINAL(Max Linear Velocity FINAL="800,800,800")
   > Set Max Linear Acceleration FINAL(Max Linear Acceleration FINAL="2000,2000,2000")
   > Set Max Angular Velocity FINAL(Max Angular Velocity FINAL=2)
   > Set Max Angular Acceleration FINAL(Max Angular Acceleration FINAL=10)
   > n41 = Random Vector Function(
    SkelMeshComponent=Get OwnerSkelMeshComp()
    JointName=@n39.JointName_27_#7
    GameTimeInSeconds=GetGameTimeInSeconds()
    TimeOffset=0
    PerlinScale=RandomFloatInRange(Min=1, Max=3)
    Frequency="5,7,12"
    Amplitude="50,70,45"
    Offset="-500,-50,0"
   )
   > n42 = Set Wind Velocity FINAL(Wind Velocity FINAL=@n41)
   n43 = Branch(execute=@n38.else, Condition=Get IcicleLarva_AirCharge())
   > Set CurrentState(CurrentState=Air Charge)
   > Set Cloth Core Settings FINAL(Cloth Core Settings FINAL=Get Cloth Core Settings AirCharge() -> (Cloth Core Settings AirCharge:Cloth Core Setting Struct Structure))
   n44 = Break Cloth Setting Struct(ClothSettingStruct=Get Cloth Settings AirCharge() -> (Cloth Settings AirCharge:Cloth Setting Struct Structure)) -> (
    LinearVelocityScale_7_#1=~
    AngularVelocityScale_6_#2=~
    FictitiousAngularScale_10_#3=~
    GravityOverride_15_#4=~
   )
   > n45 = AlignVectorWithJoint(
    Vector=@n44.GravityOverride_15_#4
    MeshComp=Get OwnerSkelMeshComp()
    JointName=@n44.JointName_27_#7
   )
   > Set Cloth Settings FINAL(Cloth Settings FINAL=Make Cloth Setting Struct(
     LinearVelocityScale_7_#1=@n44
     AngularVelocityScale_6_#2=@n44.AngularVelocityScale_6_#2
     FictitiousAngularScale_10_#3=@n44.FictitiousAngularScale_10_#3
     GravityOverride_15_#4=@n45
     DragCoefficient_23_#5=@n44.DragCoefficient_23_#5
     LiftCoefficient_24_#6=@n44.LiftCoefficient_24_#6
     JointName_27_#7=@n44.JointName_27_#7
    ))
   > Set Max Linear Velocity FINAL(Max Linear Velocity FINAL="800,800,800")
   > Set Max Linear Acceleration FINAL(Max Linear Acceleration FINAL="2000,2000,2000")
   > Set Max Angular Velocity FINAL(Max Angular Velocity FINAL=2)
   > Set Max Angular Acceleration FINAL(Max Angular Acceleration FINAL=10)
   > n46 = Random Vector Function(
    SkelMeshComponent=Get OwnerSkelMeshComp()
    JointName=@n44.JointName_27_#7
    GameTimeInSeconds=GetGameTimeInSeconds()
    TimeOffset=0
    PerlinScale=RandomFloatInRange(Min=1, Max=3)
    Frequency="5,7,12"
    Amplitude="50,70,45"
    Offset="-500,-50,0"
   )
   > n47 = Set Wind Velocity FINAL(Wind Velocity FINAL=@n46)
   n48 = Branch(execute=@n43.else, Condition=Get IcicleLarva_GroundPunch() -> (IcicleLarva_GroundPunch:bool))
   > Set CurrentState(CurrentState=Ground Punch)
   > Set Cloth Core Settings FINAL(Cloth Core Settings FINAL=Get Cloth Core Settings GroundPunch() -> (Cloth Core Settings GroundPunch:Cloth Core Setting Struct Structure))
   n49 = Break Cloth Setting Struct(ClothSettingStruct=Get Cloth Setting GroundPunch() -> (Cloth Setting GroundPunch:Cloth Setting Struct Structure)) -> (
    LinearVelocityScale_7_#1=~
    AngularVelocityScale_6_#2=~
    FictitiousAngularScale_10_#3=~
    GravityOverride_15_#4=~
   )
   > n50 = AlignVectorWithJoint(
    Vector=@n49.GravityOverride_15_#4
    MeshComp=Get OwnerSkelMeshComp()
    JointName=@n49.JointName_27_#7
   )
   > Set Cloth Settings FINAL(Cloth Settings FINAL=Make Cloth Setting Struct(
     LinearVelocityScale_7_#1=@n49
     AngularVelocityScale_6_#2=@n49.AngularVelocityScale_6_#2
     FictitiousAngularScale_10_#3=@n49.FictitiousAngularScale_10_#3
     GravityOverride_15_#4=@n50
     DragCoefficient_23_#5=@n49.DragCoefficient_23_#5
     LiftCoefficient_24_#6=@n49.LiftCoefficient_24_#6
     JointName_27_#7=@n49.JointName_27_#7
    ))
   > Set Max Linear Velocity FINAL(Max Linear Velocity FINAL="800,800,800")
   > Set Max Linear Acceleration FINAL(Max Linear Acceleration FINAL="1800,1800,1800")
   > Set Max Angular Velocity FINAL(Max Angular Velocity FINAL=2)
   > Set Max Angular Acceleration FINAL(Max Angular Acceleration FINAL=10)
   > n51 = Random Vector Function(
    SkelMeshComponent=Get OwnerSkelMeshComp()
    JointName=@n49.JointName_27_#7
    GameTimeInSeconds=GetGameTimeInSeconds()
    TimeOffset=0
    PerlinScale=1
    Frequency="3,2,1"
    Amplitude="70,130,150"
    Offset="-1000,-1000,0"
   )
   > n52 = Set Wind Velocity FINAL(Wind Velocity FINAL=@n51)
   Set CurrentState(execute=@n48.else, CurrentState=OnGround)
   Set Is LOD2-3(execute=@n10.then_9, Is LOD2-3:bool="integer >= integer"(A=Get CurrentLOD(), B=2)) -> (Output_Get:bool)
   Set IcicleLarva_Ascend(
    execute=@n10.then_0
    IcicleLarva_Ascend:bool=HasMatchingGameplayTag!(
     self:Gameplay Tag Asset Interface Interface=Get As Fort Player Pawn()
     TagToCheck:Gameplay Tag Structure=(TagName="IcicleLarva.MovementState.Ascend")
    ) -> (ReturnValue:bool)
   ) -> (Output_Get:bool)
   Set IcicleLarva_AirCharge(
    execute=@n10.then_4
    IcicleLarva_AirCharge:bool=HasMatchingGameplayTag(self=Get As Fort Player Pawn(), TagToCheck=(TagName="IcicleLarva.MovementState.AirCharge"))
   ) -> (Output_Get:bool)
   Set IcicleLarva_SprintFly(
    execute=@n10.then_3
    IcicleLarva_SprintFly:bool=HasMatchingGameplayTag(self=Get As Fort Player Pawn(), TagToCheck=(TagName="IcicleLarva.MovementState.SprintFly"))
   ) -> (Output_Get:bool)
   Set IcicleLarva_Flying(
    execute=@n10.then_2
    IcicleLarva_Flying:bool=HasMatchingGameplayTag(self=Get As Fort Player Pawn(), TagToCheck=(TagName="IcicleLarva.MovementState.Fly"))
   ) -> (Output_Get:bool)
   Set IcicleLarva_Hover(
    execute=@n10.then_1
    IcicleLarva_Hover:bool=HasMatchingGameplayTag(self=Get As Fort Player Pawn(), TagToCheck=(TagName="IcicleLarva.MovementState.Hover"))
   ) -> (Output_Get:bool)
   Set IcicleLarva_Intro(
    execute=@n10.then_5
    IcicleLarva_Intro:bool=HasMatchingGameplayTag(self=Get As Fort Player Pawn(), TagToCheck=(TagName="Granted.IcicleLarva.Intro"))
   ) -> (Output_Get:bool)
   Set IcicleLarva_HeatVision(
    execute=@n10.then_6
    IcicleLarva_HeatVision:bool=HasMatchingGameplayTag(self=Get As Fort Player Pawn(), TagToCheck=(TagName="Granted.IcicleLarva.HeatVision"))
   ) -> (Output_Get:bool)
   Set IcicleLarva_GroundPunch(
    execute=@n10.then_7
    IcicleLarva_GroundPunch:bool=HasMatchingGameplayTag(self=Get As Fort Player Pawn(), TagToCheck=(TagName="Granted.IcicleLarva.GroundPunch"))
   ) -> (Output_Get:bool)
   Outputs()
   Get Outer Drag FINAL() -> (Outer Drag FINAL:Vector 2D Structure)
   Get Outer Lift FINAL() -> (Outer Lift FINAL:Vector 2D Structure)
  ]
 }
]
variables [
 As Fort Player Pawn {type=FortPlayerPawn}
 BwdMovement {type=number}
 ChaosClothingInteractor {type=ChaosClothingInteractor, default=null}
 Cloth Core Settings AirCharge {
  type=ClothCoreSettingStruct
  default {
   collisionThickness=1
   dampingCoefficient=0.15
   frictionCoefficient=0.3
   localDampingCoefficient=0
   numberIterations=2
   numberSubsteps=1
   selfCollisionThickness=0
   animDriveDamping=<0 1>
   animDriveStiffness=<0 1>
   areaStiffness=<1 1>
   bendingStiffness=<1 1>
   edgeStiffness=<1 1>
   tetherStiffness=<1 1>
  }
 }
 Cloth Core Settings Ascend {
  type=ClothCoreSettingStruct
  default {
   collisionThickness=1
   dampingCoefficient=0.15
   frictionCoefficient=0.3
   localDampingCoefficient=0
   numberIterations=2
   numberSubsteps=1
   selfCollisionThickness=0
   animDriveDamping=<0 1>
   animDriveStiffness=<0 1>
   areaStiffness=<1 1>
   bendingStiffness=<1 1>
   edgeStiffness=<1 1>
   tetherStiffness=<1 1>
  }
 }
 Cloth Core Settings FINAL {
  type=ClothCoreSettingStruct
  default {
   collisionThickness=1.5
   dampingCoefficient=0.1
   frictionCoefficient=0.5
   localDampingCoefficient=0.1
   numberIterations=2
   numberSubsteps=1
   selfCollisionThickness=0
   animDriveDamping=<0 1>
   animDriveStiffness=<0 1>
   areaStiffness=<1 1>
   bendingStiffness=<1 1>
   edgeStiffness=<1 1>
   tetherStiffness=<1 1>
  }
 }
 Cloth Core Settings Flying {
  type=ClothCoreSettingStruct
  default {
   collisionThickness=1
   dampingCoefficient=0.15
   frictionCoefficient=0.3
   localDampingCoefficient=0
   numberIterations=2
   numberSubsteps=1
   selfCollisionThickness=0
   animDriveDamping=<0 1>
   animDriveStiffness=<0 1>
   areaStiffness=<1 1>
   bendingStiffness=<1 1>
   edgeStiffness=<1 1>
   tetherStiffness=<1 1>
  }
 }
 Cloth Core Settings GroundPunch {
  type=ClothCoreSettingStruct
  default {
   collisionThickness=1
   dampingCoefficient=0.15
   frictionCoefficient=0.3
   localDampingCoefficient=0
   numberIterations=2
   numberSubsteps=1
   selfCollisionThickness=0
   animDriveDamping=<0 1>
   animDriveStiffness=<0 1>
   areaStiffness=<1 1>
   bendingStiffness=<1 1>
   edgeStiffness=<1 1>
   tetherStiffness=<1 1>
  }
 }
 Cloth Core Settings HeatVision {
  type=ClothCoreSettingStruct
  default {
   collisionThickness=1
   dampingCoefficient=0.15
   frictionCoefficient=0.3
   localDampingCoefficient=0
   numberIterations=2
   numberSubsteps=1
   selfCollisionThickness=0
   animDriveDamping=<0 1>
   animDriveStiffness=<0 1>
   areaStiffness=<1 1>
   bendingStiffness=<1 1>
   edgeStiffness=<1 1>
   tetherStiffness=<1 1>
  }
 }
 Cloth Core Settings Hover {
  type=ClothCoreSettingStruct
  default {
   collisionThickness=1
   dampingCoefficient=0.15
   frictionCoefficient=0.3
   localDampingCoefficient=0
   numberIterations=2
   numberSubsteps=1
   selfCollisionThickness=0
   animDriveDamping=<0 1>
   animDriveStiffness=<0 1>
   areaStiffness=<1 1>
   bendingStiffness=<1 1>
   edgeStiffness=<1 1>
   tetherStiffness=<1 1>
  }
 }
 Cloth Core Settings SprintFly {
  type=ClothCoreSettingStruct
  default {
   collisionThickness=1
   dampingCoefficient=0.15
   frictionCoefficient=0.3
   localDampingCoefficient=0
   numberIterations=2
   numberSubsteps=1
   selfCollisionThickness=0
   animDriveDamping=<0 1>
   animDriveStiffness=<0 1>
   areaStiffness=<1 1>
   bendingStiffness=<1 1>
   edgeStiffness=<1 1>
   tetherStiffness=<1 1>
  }
 }
 Cloth Core Settings_Native {
  type=ClothCoreSettings
  default {
   collisionThickness=1
   dampingCoefficient=0.1
   frictionCoefficient=0.3
   localDampingCoefficient=0
   numberIterations=2
   numberSubsteps=1
   selfCollisionThickness=0
   animDriveDamping=<0 1>
   animDriveStiffness=<0 1>
   areaStiffness=<1 1>
   bendingStiffness=<1 1>
   edgeStiffness=<1 1>
   tetherStiffness=<1 1>
  }
 }
 Cloth Setting GroundPunch {
  type=ClothSettingStruct
  default {
   angularVelocityScale=1
   fictitiousAngularScale=0
   jointName=pelvis
   dragCoefficient=<0.05 0.1>
   gravityOverride=<-1000 -1000 0>
   liftCoefficient=<0.05 0.1>
   linearVelocityScale=<1 1 0>
  }
 }
 Cloth Setting HeatVision {
  type=ClothSettingStruct
  default {
   angularVelocityScale=1
   fictitiousAngularScale=0
   jointName=root
   dragCoefficient=<0.05 0.1>
   gravityOverride=<0 -1000 -980>
   liftCoefficient=<0.05 0.3>
   linearVelocityScale=<1 1 1>
  }
 }
 Cloth Settings AirCharge {
  type=ClothSettingStruct
  default {
   angularVelocityScale=1
   fictitiousAngularScale=0
   jointName=pelvis
   dragCoefficient=<0.1 5>
   gravityOverride=<-500 0 0>
   liftCoefficient=<0.1 0.7>
   linearVelocityScale=<0.2 0.2 0.2>
  }
 }
 Cloth Settings Ascend {
  type=ClothSettingStruct
  default {
   angularVelocityScale=1
   fictitiousAngularScale=0
   jointName=root
   dragCoefficient=<0.05 0.1>
   gravityOverride=<0 -400 -980>
   liftCoefficient=<0.05 0.3>
   linearVelocityScale=<1 1 1>
  }
 }
 Cloth Settings FINAL {
  type=ClothSettingStruct
  default {
   angularVelocityScale=0.6
   fictitiousAngularScale=0
   jointName=root
   dragCoefficient=<2.5 2.5>
   gravityOverride=<0 0 -980>
   liftCoefficient=<0.5 0.5>
   linearVelocityScale=<0.6 0.6 0.5>
  }
 }
 Cloth Settings Flying {
  type=ClothSettingStruct
  default {
   angularVelocityScale=1
   fictitiousAngularScale=0
   jointName=pelvis
   dragCoefficient=<0.1 5>
   gravityOverride=<-500 0 0>
   liftCoefficient=<0.1 0.7>
   linearVelocityScale=<0.2 0.2 0.2>
  }
 }
 Cloth Settings Hover {
  type=ClothSettingStruct
  default {
   angularVelocityScale=1
   fictitiousAngularScale=0
   jointName=root
   dragCoefficient=<0.05 0.1>
   gravityOverride=<0 -400 -980>
   liftCoefficient=<0.05 0.3>
   linearVelocityScale=<1 1 1>
  }
 }
 Cloth Settings SprintFly {
  type=ClothSettingStruct
  default {
   angularVelocityScale=1
   fictitiousAngularScale=0
   jointName=pelvis
   dragCoefficient=<0.1 5>
   gravityOverride=<-500 0 0>
   liftCoefficient=<0.1 0.7>
   linearVelocityScale=<0.2 0.2 0.2>
  }
 }
 CurrentState {type=string, default=""}
 DeltaTimeLocal {type=number}
 GravityOverride {type=Vector}
 IcicleLarva_AirCharge {type=boolean}
 IcicleLarva_Ascend {type=boolean}
 IcicleLarva_Flying {type=boolean}
 IcicleLarva_GroundPunch {type=boolean}
 IcicleLarva_HeatVision {type=boolean}
 IcicleLarva_Hover {type=boolean}
 IcicleLarva_Intro {type=boolean}
 IcicleLarva_SprintFly {type=boolean}
 Is LOD2-3 {type=boolean}
 Max Angular Acceleration FINAL {type=number}
 Max Angular Velocity FINAL {type=number}
 Max Linear Acceleration FINAL {type=Vector}
 Max Linear Velocity FINAL {type=Vector}
 Outer Drag FINAL {type=Vector2D, default=<0 0>}
 Outer Lift FINAL {type=Vector2D, default=<0 0>}
 OwnerSkelMeshComp {type=SkeletalMeshComponent}
 PartSkelMeshComp {type=SkeletalMeshComponent}
 RBAN Settings Out {type=FortRigidBodyAnimNodeParameters}
 Skel_Mesh {type=SkeletalMeshComponent}
 TrailControllerRot {type=Rotator}
 Wind Gust Local Settings Native {type=WindGustSettings, default {elapsedTime=0, finalGustStrengthLocal=0, gustDuration=0, gustStrength=0, timeBetweenGusts=0}}
 Wind Gust Local Settings Out {type=WindGustSettings, default {elapsedTime=0, finalGustStrengthLocal=0, gustDuration=0, gustStrength=0, timeBetweenGusts=0}}
 Wind Velocity FINAL {type=Vector}
]
