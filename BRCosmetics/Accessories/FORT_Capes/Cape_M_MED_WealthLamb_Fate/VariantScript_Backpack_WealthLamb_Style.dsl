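; Variant script for the WealthLamb Fate back bling that picks its material variant tag from the player's cosmetic loadout.
; It checks whether the matching outfit is equipped, then compares Progressive-channel stages on outfit and back bling.
; Matching base stages return Mat1 and matching toon stages return Mat2.
; Without the outfit, a non-toon back bling returns an empty tag; unresolved branches fall through to Mat3.
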(fn DetermineLoadoutVariantSelection
  (params (Loadout :type Cosmetic%20Loadout%20Structure%20%28by%20ref%29))
  (if (Class|FortCosmeticFunctionLibrary|IsItemInLoadout Loadout "/BRCosmetics/Athena/Items/Cosmetics/Characters/Character_WealthLambFate.Character_WealthLambFate")
    ; base super outfit + base super backbling
    (if (== (Class|FortCosmeticFunctionLibrary|GetActiveVariantForSoftItemAndChannel Loadout "/BRCosmetics/Athena/Items/Cosmetics/Backpacks/Backpack_WealthLambFate.Backpack_WealthLambFate" "(TagName=\"Cosmetics.Variant.Channel.Progressive\")") "(TagName=\"Cosmetics.Variant.Property.Stage1\")")
      (if (== (Class|FortCosmeticFunctionLibrary|GetActiveVariantForSoftItemAndChannel Loadout "/BRCosmetics/Athena/Items/Cosmetics/Characters/Character_WealthLambFate.Character_WealthLambFate" "(TagName=\"Cosmetics.Variant.Channel.Progressive\")") "(TagName=\"Cosmetics.Variant.Property.Stage1\")")
        (switch Utilities|FlowControl|Switch|SwitchonGameplayTag 0
          (:Default
            (return "(TagName=\"Cosmetics.Variant.Property.Mat1\")"))))
      (elif (== (Class|FortCosmeticFunctionLibrary|GetActiveVariantForSoftItemAndChannel Loadout "/BRCosmetics/Athena/Items/Cosmetics/Backpacks/Backpack_WealthLambFate.Backpack_WealthLambFate" "(TagName=\"Cosmetics.Variant.Channel.Progressive\")") "(TagName=\"Cosmetics.Variant.Property.Stage2\")")
        ; toon super outfit + toon super Backbling
        (if (== (Class|FortCosmeticFunctionLibrary|GetActiveVariantForSoftItemAndChannel Loadout "/BRCosmetics/Athena/Items/Cosmetics/Characters/Character_WealthLambFate.Character_WealthLambFate" "(TagName=\"Cosmetics.Variant.Channel.Progressive\")") "(TagName=\"Cosmetics.Variant.Property.Stage2\")")
          (switch Utilities|FlowControl|Switch|SwitchonGameplayTag 0
            (:Default
              (return "(TagName=\"Cosmetics.Variant.Property.Mat2\")")))
          (elif (== (Class|FortCosmeticFunctionLibrary|GetActiveVariantForSoftItemAndChannel Loadout "/BRCosmetics/Athena/Items/Cosmetics/Backpacks/Backpack_WealthLambFate.Backpack_WealthLambFate" "(TagName=\"Cosmetics.Variant.Channel.Progressive\")") "(TagName=\"Cosmetics.Variant.Property.Stage2\")")
            ; base super outfit + toon super Backbling
            (if (== (Class|FortCosmeticFunctionLibrary|GetActiveVariantForSoftItemAndChannel Loadout "/BRCosmetics/Athena/Items/Cosmetics/Characters/Character_WealthLambFate.Character_WealthLambFate" "(TagName=\"Cosmetics.Variant.Channel.Progressive\")") "(TagName=\"Cosmetics.Variant.Property.Stage1\")")
              _)))))
    (elif (== (Class|FortCosmeticFunctionLibrary|GetActiveVariantForSoftItemAndChannel Loadout "/BRCosmetics/Athena/Items/Cosmetics/Backpacks/Backpack_WealthLambFate.Backpack_WealthLambFate" "(TagName=\"Cosmetics.Variant.Channel.Progressive\")") "(TagName=\"Cosmetics.Variant.Property.Stage2\")")
      _
      (else
        (switch Utilities|FlowControl|Switch|SwitchonGameplayTag 0
          (:Default
            (return "(TagName=\"\")"))))))
  (switch Utilities|FlowControl|Switch|SwitchonGameplayTag (Class|FortCosmeticFunctionLibrary|GetActiveVariantForSoftItemAndChannel Loadout "/BRCosmetics/Athena/Items/Cosmetics/Backpacks/Backpack_WealthLambFate.Backpack_WealthLambFate" "(TagName=\"Cosmetics.Variant.Channel.Progressive\")")
    (:Default
      (return "(TagName=\"Cosmetics.Variant.Property.Mat3\")"))))
